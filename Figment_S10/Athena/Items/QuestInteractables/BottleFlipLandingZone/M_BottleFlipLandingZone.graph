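# Deferred decal material for the bottle flip landing zone quest interactable in Figment Season 10, using the Flippin' Bottle toy icon texture.
# It uses translucent blending and Default Lit shading, with a small graph of one texture sample and two scalar constants.
# The file does not record which material inputs these three nodes feed.
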Material /Figment_S10/Athena/Items/QuestInteractables/BottleFlipLandingZone/M_BottleFlipLandingZone
graph {
 properties {blendMode=BLEND_Translucent, materialDomain=MD_DeferredDecal, pixelDepthOffsetMode=PDOM_Legacy, usedShadingModels=Default Lit}
 nodes [
  Constant()
  Constant() {r=0.5}
  TextureSample() {texture=/Figment_S10/2dAssets/Toys/Season10/T-Icons_Toys_Season10-FlippinBottle}
 ]
}
