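; Quest landing-zone prop that credits the bottle flip target quest when a toy bottle flip succeeds within its overlap sphere.
; On clients, BeginPlay hides it unless the first local player's quest is active, with a testing bypass.
; On the server, sphere overlap binds to the bottle's Finished event, and a success sends a tagged stat event to the bottle owner's quest manager.
; Bindings clear when the flip finishes or overlap ends.
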(event EventBeginPlay
  (bind _returnvalue (Fortnite|GetLocalFortPlayerControllers true))
  (bind _output (Utilities|Array|Get(acopy) _returnvalue))
  (bind _returnvalue_2 (Quests|GetQuestManager _output "Invalid"))
  (bind _questobjective (Variables|Default|GetQuestObjective))
  (bind _returnvalue_3 (Quests|GetQuestwithDefinition _returnvalue_2 _questobjective))
  ;[1 1 1] Hide on client if player doens't have quest
  (|CallParentFunction)
  (Networking|SwitchHasAuthority
    (:Remote
      ;;[1 1 1] Show up if testing
      (if (not (Variables|Default|GetEnableTesting))
        (Utilities|IsValid _returnvalue_3
          (:Is%20Valid
            (Variables|Default|SetQuestActive (Utilities|Enum|Equal(Enum) (Quest|GetFortQuestState _returnvalue_3) "Active")))
          (:Is%20Not%20Valid
            (Variables|Default|SetQuestActive false)))
        (Rendering|SetActorHiddenInGame (not (Variables|Default|GetQuestActive)))))))

(event OnComponentBeginOverlap(OverlapSphere)
  (params (OverlappedComponent :type Primitive%20Component%20Object%20Reference) (OtherActor :type Actor%20Object%20Reference) (OtherComp :type Primitive%20Component%20Object%20Reference) (OtherBodyIndex :type Integer) (bFromSweep :type Boolean) (SweepResult :type Hit%20Result%20Structure%20%28by%20ref%29))
  ;[1 1 1] Bind to bottle flip finished event
  (Networking|SwitchHasAuthority
    (:Authority
      (bind _asb_toy_bottle_base (Utilities|Casting|CastToB_ToyBottle_Base OtherActor))
      (EventDispatchers|BindEventtoFinished _asb_toy_bottle_base (delegate Custom|BottleFlipFinished)))))

(event Custom|BottleFlipFinished
  (params (Success :type Boolean) (BottleRef :type B%20Toy%20Bottle%20Base%20Object%20Reference))
  ;[1 1 1] Progress quest if flip succeeds
  (if Success
    (Quest|SendComplexCustomStatEvent (Quests|GetQuestManager (Fortnite|GetFortPlayerControllerfromActor (Actor|GetOwner BottleRef)) "Invalid") 0 "(GameplayTags=((TagName=\"Athena.Quests.BottleFlipTarget\")))" "(GameplayTags=)" 1)
    (else
      _))
  (EventDispatchers|UnbindallEventsfromFinished BottleRef))

(event OnComponentEndOverlap(OverlapSphere)
  (params (OverlappedComponent :type Primitive%20Component%20Object%20Reference) (OtherActor :type Actor%20Object%20Reference) (OtherComp :type Primitive%20Component%20Object%20Reference) (OtherBodyIndex :type Integer))
  ;[1 1 1] Clean up event binds
  (Networking|SwitchHasAuthority
    (:Authority
      (bind _asb_toy_bottle_base (Utilities|Casting|CastToB_ToyBottle_Base OtherActor))
      (EventDispatchers|UnbindallEventsfromFinished _asb_toy_bottle_base))))

(fn ConstructionScript
  (|CallParentFunction))
